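# Editor utility widget that generates cosmetic variant token (VTID) assets for item definitions picked from the Content Browser.
# Embedded Python reads each item's mesh, material and particle variant channels, looping over options and skipping those that start unlocked.
# GetCreateVTIDAsset reuses or creates tokens named by item and option-index letter, then the graph fills names, tags, release version and equip settings.
# It logs results to a text box; the icon-component script is unwired.
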EditorUtilityWidgetBlueprint /Game/NeverCook/EditorUtilities/CosmeticTools/VariantIDSetup/EUW_CosmeticVariantIDSetup
graphs [
 EventGraph {nodes [
   "On Clicked (Button_GenerateVTIDs)"() {is=title K2Node_ComponentBoundEvent}
   > n1 = For Each Loop(Array:Array of Fort Item Definition Object References=Get IDs() -> (IDs:Array of Fort Item Definition Object References)) -> (Array Element:Fort Item Definition Object Reference)
   n2 = GetDisplayName(Object=@n1)
   > Set Result(Result:Text=Format Text!(Format="LOCTEXT(\"{1}\\r\\n\\r\\n{2}\\r\\n\")", 1:Text=Get Result() -> (Result:Text), 2:Text="To Text (String)"(InString=@n2)) -> (Result)) -> (Output_Get:Text)
   > n3 = Execute Python Script!(
    execute
    PythonScript="import unreal\r\n\r\ncount = len(item_def.get_editor_property(\"item_variants\"))"
    item_def:Fort Item Definition Object Reference=@n1
   ) -> (then, ReturnValue, Count:int) {inputs [item_def], outputs [Count]}
   > n4 = For Loop(LastIndex=int - int(A=@n3.Count))
   > n5 = Execute Python Script!(
    execute
    PythonScript="import unreal\r\n\r\nchannel = item_def.get_editor_property(\"item_variants\")[variant_index]\r\nchannel_class = channel.get_class().get_name()\r\n\r\nOPTIONS_PROPERTY = {\r\n    \"FortCosmeticMeshVariant\":     \"mesh_options\",\r\n    \"FortCosmeticMaterialVariant\": \"material_options\",\r\n    \"FortCosmeticParticleVariant\": \"particle_options\",\r\n}\r\n\r\nchannel_class_output   = OPTIONS_PROPERTY.get(channel_class)\r\nvariant_options_length = len(channel.get_editor_property(channel_class_output))"
    item_def=@n1
    variant_index:int=@n4
   ) -> (then, ReturnValue, variant_options_length:int) {inputs [item_def, variant_index], outputs [variant_options_length]}
   > n6 = For Loop(LastIndex=int - int(A=@n5.variant_options_length))
   > n7 = Execute Python Script!(
    execute
    PythonScript="import unreal\r\n\r\nchannel = item_def.get_editor_property(\"item_variants\")[channel_index]\r\n\r\nchannel_tag = channel.get_editor_property(\"variant_channel_tag\")\r\nchannel_output = str(channel_tag.to_tuple()[0]) \r\n\r\nchannel_class = channel.get_class().get_name()\r\n\r\nOPTIONS_PROPERTY = {\r\n    \"FortCosmeticMeshVariant\":     \"mesh_options\",\r\n    \"FortCosmeticMaterialVariant\": \"material_options\",\r\n    \"FortCosmeticParticleVariant\": \"particle_options\",\r\n}\r\n\r\nchannel_class_output   = OPTIONS_PROPERTY.get(channel_class)\r\n\r\nvariant_option_output = str(item_def.get_editor_property(\"item_variants\")[channel_index].get_editor_property(channel_class_output)[variant_index].get_editor_property(\"customization_variant_tag\").to_tuple()[0])\r\n\r\nstart_unlocked = item_def.get_editor_property(\"item_variants\")[channel_index].get_editor_property(channel_class_output)[variant_index].get_editor_property(\"bStartUnlocked\")"
    item_def=@n1
    variant_index=@n6
    channel_index:int=@n4
   ) -> (then, ReturnValue, variant_option_output:String, channel_output:String, start_unlocked:bool=~) {inputs [item_def, variant_index, channel_index], outputs [variant_option_output, channel_output, start_unlocked]}
   > Set Result(Result=Format Text(
     Format="LOCTEXT(\"{1}{2}\")"
     1=Get Result()
     2=Format Text!(
      Format="LOCTEXT(\"{channel_output}\\r\\n{varian_tag_output}\\r\\n{start_unlocked}\\r\\n\")"
      channel_output:String=@n7.channel_output
      varian_tag_output:String=@n7.variant_option_output
      start_unlocked:bool=@n7.start_unlocked
     ) -> (Result)
    ))
   > Branch(Condition=NOT Boolean(A=@n7.start_unlocked))
   n8 = Select!(
    Option 0:String=A
    Option 1:String=B
    Option 2:String=C
    Option 3:String=D
    Option 4:String=E
    Option 5:String=F
    Option 6:String=G
    Option 7:String=H
    Option 8:String=I
    Option 9:String=J
    Option 10:String=K
    Option 11:String=L
    Option 12:String=M
    Option 13:String=N
    Option 14:String=O
    Option 15:String=P
    Index:int=@n6
   ) -> (ReturnValue:String)
   > n9 = GetCreateVTIDAsset!(execute, self, AssetName:String=Append!(A=VTID_, B=GetDisplayName(Object=@n1), C:String=_Style, D:String=@n8) -> (ReturnValue)) -> (then, VariantToken:Object Reference)
   > n10 = Cast To FortVariantTokenType!(execute, Object=@n9) -> (then, CastFailed, AsFort Variant Token Type:Fort Variant Token Type Object Reference) {is=title K2Node_DynamicCast}
   n11 = MakeLiteralText(Value="LOCTEXT(\"Style\")")
   > SetRawShortDescription!(execute, self:Fort Item Definition Object Reference=@n10, InShortDescription:"Text (by ref)"=@n11) -> (then)
   > SetRawDisplayName!(
    execute
    self:Fort Item Definition Object Reference=@n10
    InDisplayName:"Text (by ref)"=Format Text!(Format="LOCTEXT(\"{1}{2}\")", 1:Text=@n11, 2:String=@n8) -> (Result)
   ) -> (then)
   > n12 = RequestGameplayTag!(execute, TagName:String=@n7.variant_option_output, bErrorIfNotFound:bool=false) -> (then, OutTag:Gameplay Tag Structure, ReturnValue:bool)
   > SetEditorProperty(Object=@n10, PropertyName=VariantNameTag, PropertyValue:Gameplay Tag Structure=@n12)
   > n13 = RequestGameplayTag(TagName=@n7.channel_output, bErrorIfNotFound=false)
   > SetEditorProperty(Object=@n10, PropertyName=VariantChannelTag, PropertyValue=@n13)
   > n14 = EditorOnly_GetItemDefinitionFortReleaseVersion!(execute, ItemDefinition:Item Definition Base Object Reference=@n1) -> (then, ReturnValue:Fort Release Version Structure)
   > SetEditorProperty(Object=@n10, PropertyName=FortReleaseVersion, PropertyValue:Fort Release Version Structure=@n14)
   > SetEditorProperty(Object=@n10, PropertyName=bGrantBaseItem, PropertyValue:bool=MakeLiteralBool(Value=true))
   > SetEditorProperty(Object=@n10, PropertyName=bAutoEquipVariant, PropertyValue=MakeLiteralBool(Value="Equal (Integer)"(A=@n6)))
   > SetEditorProperty(Object=@n10, PropertyName=cosmetic_item, PropertyValue:Fort Item Definition Object Reference=@n1)
   > SetEditorProperty(Object=@n10, PropertyName=VariantEquipRules, PropertyValue:Byte=MakeLiteralByte!(Value:Byte=3) -> (ReturnValue:Byte=0))
   "SetText (Text Box)"!(
    execute=@n1.Completed
    self:Text Box Object Reference=Get EditorUtilityEditableTextBox() -> (EditorUtilityEditableTextBox:Editor Utility Editable Text Box Object Reference)
    InText:Text=Get Result()
   ) -> (then)
   > SyncBrowserToFolders!(execute, FolderList:Array of Strings=Make Array("[0]":String=/BRCosmetics/Athena/Items/CosmeticVariantTokens) -> (Array:Array of Strings)) -> (then)
   "On Clicked (Button_GenerateVTIDs_1)"() {is=title K2Node_ComponentBoundEvent}
   > Clear(TargetArray:Array of Fort Item Definition Object References=Get IDs())
   > n15 = GetSelectedAssets()
   > n16 = For Each Loop(Array:Array of Object References=@n15) -> (Array Element:Object Reference)
   > n17 = Cast To FortItemDefinition!(execute, Object=@n16) -> (then, CastFailed, AsFort Item Definition:Fort Item Definition Object Reference) {is=title K2Node_DynamicCast}
   > Add(TargetArray:Array of Fort Item Definition Object References=Get IDs(), NewItem:"Fort Item Definition Object Reference (by ref)"=@n17) {is=title K2Node_CallArrayFunction}
   Event PreConstruct()
   > SetObject(self=Get DetailsView_162() -> (DetailsView_162:Details View Object Reference), NewObject=Self-Reference())
   > SetObject(
    self=Get DetailsView() -> (DetailsView:Details View Object Reference)
    NewObject="Get (a copy)"(Array:Array of Fort Item Definition Object References=Get IDs()) -> (Output:Fort Item Definition Object Reference)
   )
   n18 = Execute Python Script!(
    execute
    PythonScript="import unreal\r\n\r\ncount = len(item_def.get_editor_property(\"item_variants\"))"
    item_def="Get (a copy)"(Array=Get IDs())
   ) -> (then, ReturnValue, Count:int) {inputs [item_def], outputs [Count]}
   > "SetText (Text Box)"(self=Get EditorUtilityEditableTextBox(), InText="To Text (Integer)"(Value=MakeLiteralInt!(Value:int=@n18.Count) -> (ReturnValue:int=0)))
   n19 = Execute Python Script!(
    execute
    PythonScript="import unreal\r\n\r\nchannel = item_def.get_editor_property(\"item_variants\")[variant_index]\r\nchannel_class = channel.get_class().get_name()\r\n\r\nOPTIONS_PROPERTY = {\r\n    \"FortCosmeticMeshVariant\":     \"mesh_options\",\r\n    \"FortCosmeticMaterialVariant\": \"material_options\",\r\n    \"FortCosmeticParticleVariant\": \"particle_options\",\r\n}\r\n\r\nchannel_class_output   = OPTIONS_PROPERTY.get(channel_class)"
    item_def="Get (a copy)"(Array=Get IDs())
    variant_index:int=MakeLiteralInt(Value=0) -> (ReturnValue=0)
   ) -> (then, ReturnValue, channel_class_output:String) {inputs [item_def, variant_index], outputs [channel_class_output]}
   > "SetText (Text Box)"(self=Get EditorUtilityEditableTextBox(), InText="To Text (String)"(InString=@n19.channel_class_output))
   n20 = Execute Python Script!(
    execute
    PythonScript="import unreal\r\n\r\nchannel = item_def.get_editor_property(\"item_variants\")[variant_index]\r\nchannel_class = channel.get_class().get_name()\r\n\r\nOPTIONS_PROPERTY = {\r\n    \"FortCosmeticMeshVariant\":     \"mesh_options\",\r\n    \"FortCosmeticMaterialVariant\": \"material_options\",\r\n    \"FortCosmeticParticleVariant\": \"particle_options\",\r\n}\r\n\r\nchannel_class_output   = OPTIONS_PROPERTY.get(channel_class)\r\nvariant_options_length = len(channel.get_editor_property(channel_class_output))"
    item_def="Get (a copy)"(Array=Get IDs())
    variant_index=MakeLiteralInt(Value=0) -> (ReturnValue=0)
   ) -> (then, ReturnValue, variant_options_length:int) {inputs [item_def, variant_index], outputs [variant_options_length]}
   > "SetText (Text Box)"(
    self=Get EditorUtilityEditableTextBox()
    InText="To Text (Integer)"(Value=MakeLiteralInt(Value=@n20.variant_options_length) -> (ReturnValue=0))
   )
   n21 = Execute Python Script!(
    execute
    PythonScript="import unreal\r\n\r\nchannel = item_def.get_editor_property(\"item_variants\")[channel_index]\r\n\r\nchannel_tag = channel.get_editor_property(\"variant_channel_tag\")\r\nchannel_output = str(channel_tag.to_tuple()[0]) \r\n\r\nchannel_class = channel.get_class().get_name()\r\n\r\nOPTIONS_PROPERTY = {\r\n    \"FortCosmeticMeshVariant\":     \"mesh_options\",\r\n    \"FortCosmeticMaterialVariant\": \"material_options\",\r\n    \"FortCosmeticParticleVariant\": \"particle_options\",\r\n}\r\n\r\nchannel_class_output   = OPTIONS_PROPERTY.get(channel_class)\r\n\r\nvariant_option_output = str(item_def.get_editor_property(\"item_variants\")[channel_index].get_editor_property(channel_class_output)[variant_index].get_editor_property(\"customization_variant_tag\").to_tuple()[0])\r\n\r\nstart_unlocked = item_def.get_editor_property(\"item_variants\")[channel_index].get_editor_property(channel_class_output)[variant_index].get_editor_property(\"bStartUnlocked\")"
    item_def="Get (a copy)"(Array=Get IDs())
    channel_index:int=MakeLiteralInt(Value=0) -> (ReturnValue=0)
    variant_index=MakeLiteralInt(Value=1) -> (ReturnValue=0)
   ) -> (then, ReturnValue, variant_option_output:String, channel_output:String, start_unlocked:bool=~) {inputs [item_def, channel_index, variant_index], outputs [variant_option_output, channel_output, start_unlocked]}
   > "SetText (Text Box)"(
    self=Get EditorUtilityEditableTextBox()
    InText=Format Text!(
     Format="LOCTEXT(\"{channel_output}\\r\\n{varian_tag_output}\\r\\n{start_unlocked}\")"
     channel_output:String=@n21.channel_output
     varian_tag_output:String=@n21.variant_option_output
     start_unlocked:bool=@n21.start_unlocked
    ) -> (Result)
   )
   n22 = For Each Loop(Array:Array of Fort Item Definition Object References=Get IDs()) -> (Array Element:Fort Item Definition Object Reference)
   "SetText (Text Box)"(execute=@n22.Completed, self=Get EditorUtilityEditableTextBox(), InText=Get Result())
   n23 = GetDisplayName(Object=@n22)
   Set Result(
    execute=@n22.LoopBody
    Result=Format Text!(Format="LOCTEXT(\"{1}\\r\\n\\r\\n{2}\\r\\n\")", 1:Text=Get Result(), 2:Text="To Text (String)"(InString=@n23)) -> (Result)
   )
   > n24 = Execute Python Script!(execute, PythonScript="import unreal\r\n\r\ncount = len(item_def.get_editor_property(\"item_variants\"))", item_def=@n22) -> (then, ReturnValue, Count:int) {inputs [item_def], outputs [Count]}
   > n25 = For Loop(LastIndex=int - int(A=@n24.Count))
   > n26 = Execute Python Script!(
    execute
    PythonScript="import unreal\r\n\r\nchannel = item_def.get_editor_property(\"item_variants\")[variant_index]\r\nchannel_class = channel.get_class().get_name()\r\n\r\nOPTIONS_PROPERTY = {\r\n    \"FortCosmeticMeshVariant\":     \"mesh_options\",\r\n    \"FortCosmeticMaterialVariant\": \"material_options\",\r\n    \"FortCosmeticParticleVariant\": \"particle_options\",\r\n}\r\n\r\nchannel_class_output   = OPTIONS_PROPERTY.get(channel_class)\r\nvariant_options_length = len(channel.get_editor_property(channel_class_output))"
    item_def=@n22
    variant_index:int=@n25
   ) -> (then, ReturnValue, variant_options_length:int) {inputs [item_def, variant_index], outputs [variant_options_length]}
   > n27 = For Loop(LastIndex=int - int(A=@n26.variant_options_length))
   > n28 = Execute Python Script!(
    execute
    PythonScript="import unreal\r\n\r\nchannel = item_def.get_editor_property(\"item_variants\")[channel_index]\r\n\r\nchannel_tag = channel.get_editor_property(\"variant_channel_tag\")\r\nchannel_output = str(channel_tag.to_tuple()[0]) \r\n\r\nchannel_class = channel.get_class().get_name()\r\n\r\nOPTIONS_PROPERTY = {\r\n    \"FortCosmeticMeshVariant\":     \"mesh_options\",\r\n    \"FortCosmeticMaterialVariant\": \"material_options\",\r\n    \"FortCosmeticParticleVariant\": \"particle_options\",\r\n}\r\n\r\nchannel_class_output   = OPTIONS_PROPERTY.get(channel_class)\r\n\r\nvariant_option_output = str(item_def.get_editor_property(\"item_variants\")[channel_index].get_editor_property(channel_class_output)[variant_index].get_editor_property(\"customization_variant_tag\").to_tuple()[0])\r\n\r\nstart_unlocked = item_def.get_editor_property(\"item_variants\")[channel_index].get_editor_property(channel_class_output)[variant_index].get_editor_property(\"bStartUnlocked\")"
    item_def=@n22
    variant_index=@n27
    channel_index:int=@n25
   ) -> (then, ReturnValue, variant_option_output:String, channel_output:String, start_unlocked:bool=~) {inputs [item_def, variant_index, channel_index], outputs [variant_option_output, channel_output, start_unlocked]}
   > Set Result(Result=Format Text(
     Format="LOCTEXT(\"{1}{2}\")"
     1=Get Result()
     2=Format Text!(
      Format="LOCTEXT(\"{channel_output}\\r\\n{varian_tag_output}\\r\\n{start_unlocked}\\r\\n\")"
      channel_output:String=@n28.channel_output
      varian_tag_output:String=@n28.variant_option_output
      start_unlocked:bool=@n28.start_unlocked
     ) -> (Result)
    ))
   Execute Python Script!(
    execute
    PythonScript="import unreal\r\n\r\nchannel = item_def.get_editor_property(\"item_variants\")[variant_index]\r\nchannel_class = channel.get_class().get_name()\r\n\r\nOPTIONS_PROPERTY = {\r\n    \"FortCosmeticMeshVariant\":     \"mesh_options\",\r\n    \"FortCosmeticMaterialVariant\": \"material_options\",\r\n    \"FortCosmeticParticleVariant\": \"particle_options\",\r\n}\r\n\r\nchannel_class_output   = OPTIONS_PROPERTY.get(channel_class)\r\nvariant_option = channel.get_editor_property(channel_class_output)\r\nvariant_option_tag = variant_option.get_editor_property(\"customization_variant_tag\")\r\nvariant_option_output = str(variant_option_tag.to_tuple()[0])"
    item_def:Wildcard
    variant_index:Wildcard
   ) -> (then, ReturnValue, variant_option_output:Wildcard) {inputs [item_def, variant_index], outputs [variant_option_output]}
   Execute Python Script!(
    execute
    PythonScript="import unreal\r\n\r\n## Build the two new instanced structs\r\nicon_data = unreal.InstancedStruct()\r\nicon_data.set_struct_type(unreal.ItemComponentData_Icon)\r\n\r\nlarge_icon_data = unreal.InstancedStruct()\r\nlarge_icon_data.set_struct_type(unreal.FortItemComponentData_LargeIcon)\r\n\r\n## Append to existing dataList\r\ndata_list = list(item_def.get_editor_property(\"data_list\"))\r\ndata_list.append(icon_data)\r\ndata_list.append(large_icon_data)\r\n\r\nitem_def.set_editor_property(\"data_list\", data_list)\r\nunreal.EditorAssetLibrary.save_loaded_asset(item_def)\r\n"
    item_def:Fort Variant Token Type Object Reference=@n10
   ) -> (then, ReturnValue) {inputs [item_def]}
  ]}
 GetCreateVTIDAsset {nodes [
   n1 = GetCreateVTIDAsset!() -> (then, AssetName:String) {is=title K2Node_FunctionEntry}
   > n2 = Get Asset By Object Path!(
    execute
    self:Asset Registry Interface=GetAssetRegistry()
    ObjectPath_PathString:String=Append(A=&CosmeticVariantTokens/, B=@n1, C=., D=@n1)
    bIncludeOnlyOnDiskAssets:bool=false
    bSkipARFilteredAssets:bool=true
   ) -> (then, ReturnValue:Asset Data Structure) {is=title K2Node_CallFunction}
   n3 = GetAsset(InAssetData=@n2)
   > n4 = Branch(Condition=IsValid(Object=@n3))
   > Return Node!(execute, VariantToken:Object Reference=@n3)
   n5 = GetAssetTools(execute=@n4.else)
   > n6 = CreateAsset(
    self=@n5
    AssetName=@n1
    PackagePath=&CosmeticVariantTokens
    AssetClass=/Script/FortniteGame.FortVariantTokenType
    bOverwriteExisting=true
   )
   > SaveLoadedAsset(AssetToSave=@n6, bOnlyIfIsDirty=true)
   > Return Node(VariantToken=@n6)
  ]}
]
variables [
 IDs {type="array<FortItemDefinition>", default []}
 NewVar {type=boolean}
 Result {type=string, default=""}
]
widgets {
 info {rootWidgetClass=&VerticalBox, widgetCount=15}
 widgets [
  VerticalBox_31 {widgetClassPath=&VerticalBox, widget=VerticalBox_31}
  TextBlock_181 {
   widgetClassPath=&TextBlock
   parent=VerticalBox_31
   widget=TextBlock_181
   properties {text=Cosmetic Variant ID Setup}
   slot {class=VerticalBoxSlot, properties {padding=<ltrb 0 0 0 10>}}
  }
  HorizontalBox_116 {widgetClassPath=&HorizontalBox, parent=VerticalBox_31, widget=HorizontalBox_116, slot {class=VerticalBoxSlot}}
  Button_GenerateVTIDs_1 {
   widgetClassPath=&EditorUtilityButton
   parent=HorizontalBox_116
   widget=Button_GenerateVTIDs_1
   slot {class=HorizontalBoxSlot}
  }
  TextBlock_1 {
   widgetClassPath=&TextBlock
   parent=Button_GenerateVTIDs_1
   widget=TextBlock_1
   properties {
    justification=Center
    text="Get IDs \r\nFrom Browser"
    font {
     bForceMonospaced=false
     bMaterialIsStencil=false
     fontMaterial=null
     fontObject=&Roboto
     letterSpacing=0
     monospacedWidth=1
     size=15
     skewAmount=0
     typefaceFontName=Bold
     outlineSettings {
      bApplyOutlineToDropShadows=false
      bMiteredCorners=false
      bSeparateFillAlpha=false
      outlineMaterial=null
      outlineSize=0
      outlineColor=<rgba 0 0 0 1>
     }
    }
   }
   slot {class=ButtonSlot}
  }
  SizeBox_0 {
   widgetClassPath=&SizeBox
   parent=HorizontalBox_116
   widget=SizeBox_0
   properties {bOverride_MaxDesiredHeight=true, maxDesiredHeight=290}
   slot {class=HorizontalBoxSlot}
  }
  AthenaScrollBox_0 {
   widgetClassPath=/Script/FortniteUI.AthenaScrollBox
   parent=SizeBox_0
   widget=AthenaScrollBox_0
   properties {
    widgetBarStyle {
     thickness=8
     draggedThumbImage {
      bIsDynamicallyLoaded=false
      drawAs=RoundedBox
      imageType=NoImage
      mirroring=NoMirror
      resourceName=None
      resourceObject=null
      tiling=NoTile
      imageSize=<0 0>
      margin=<ltrb 0 0 0 0>
      outlineSettings {
       bUseBrushTransparency=false
       roundingType=FixedRadius
       width=0
       color {colorUseRule=UseColor_Specified, specifiedColor=<rgba 0 0 0 0>}
       cornerRadii=<4 4 4 4>
      }
      tintColor {colorUseRule=UseColor_Specified, specifiedColor=<rgba 0.2159 0.2159 0.2159 1>}
      uVRegion {bIsValid=false, max=<0 0>, min=<0 0>}
     }
     horizontalBackgroundImage {
      bIsDynamicallyLoaded=false
      drawAs=NoDrawType
      imageType=NoImage
      mirroring=NoMirror
      resourceName=None
      resourceObject=null
      tiling=NoTile
      imageSize=<0 0>
      margin=<ltrb 0 0 0 0>
      outlineSettings {
       bUseBrushTransparency=false
       roundingType=HalfHeightRadius
       width=0
       color {colorUseRule=UseColor_Specified, specifiedColor=<rgba 0 0 0 0>}
       cornerRadii=<0 0 0 0>
      }
      tintColor {colorUseRule=UseColor_Specified, specifiedColor=<rgba 1 1 1 1>}
      uVRegion {bIsValid=false, max=<0 0>, min=<0 0>}
     }
     horizontalBottomSlotImage {
      bIsDynamicallyLoaded=false
      drawAs=NoDrawType
      imageType=NoImage
      mirroring=NoMirror
      resourceName=None
      resourceObject=null
      tiling=NoTile
      imageSize=<0 0>
      margin=<ltrb 0 0 0 0>
      outlineSettings {
       bUseBrushTransparency=false
       roundingType=HalfHeightRadius
       width=0
       color {colorUseRule=UseColor_Specified, specifiedColor=<rgba 0 0 0 0>}
       cornerRadii=<0 0 0 0>
      }
      tintColor {colorUseRule=UseColor_Specified, specifiedColor=<rgba 1 1 1 1>}
      uVRegion {bIsValid=false, max=<0 0>, min=<0 0>}
     }
     horizontalTopSlotImage {
      bIsDynamicallyLoaded=false
      drawAs=NoDrawType
      imageType=NoImage
      mirroring=NoMirror
      resourceName=None
      resourceObject=null
      tiling=NoTile
      imageSize=<0 0>
      margin=<ltrb 0 0 0 0>
      outlineSettings {
       bUseBrushTransparency=false
       roundingType=HalfHeightRadius
       width=0
       color {colorUseRule=UseColor_Specified, specifiedColor=<rgba 0 0 0 0>}
       cornerRadii=<0 0 0 0>
      }
      tintColor {colorUseRule=UseColor_Specified, specifiedColor=<rgba 1 1 1 1>}
      uVRegion {bIsValid=false, max=<0 0>, min=<0 0>}
     }
     hoveredThumbImage {
      bIsDynamicallyLoaded=false
      drawAs=RoundedBox
      imageType=NoImage
      mirroring=NoMirror
      resourceName=None
      resourceObject=null
      tiling=NoTile
      imageSize=<0 0>
      margin=<ltrb 0 0 0 0>
      outlineSettings {
       bUseBrushTransparency=false
       roundingType=FixedRadius
       width=0
       color {colorUseRule=UseColor_Specified, specifiedColor=<rgba 0 0 0 0>}
       cornerRadii=<4 4 4 4>
      }
      tintColor {colorUseRule=UseColor_Specified, specifiedColor=<rgba 0.2159 0.2159 0.2159 1>}
      uVRegion {bIsValid=false, max=<0 0>, min=<0 0>}
     }
     normalThumbImage {
      bIsDynamicallyLoaded=false
      drawAs=RoundedBox
      imageType=NoImage
      mirroring=NoMirror
      resourceName=None
      resourceObject=null
      tiling=NoTile
      imageSize=<0 0>
      margin=<ltrb 0 0 0 0>
      outlineSettings {
       bUseBrushTransparency=false
       roundingType=FixedRadius
       width=0
       color {colorUseRule=UseColor_Specified, specifiedColor=<rgba 0 0 0 0>}
       cornerRadii=<4 4 4 4>
      }
      tintColor {colorUseRule=UseColor_Specified, specifiedColor=<rgba 0.0953 0.0953 0.0953 1>}
      uVRegion {bIsValid=false, max=<0 0>, min=<0 0>}
     }
     verticalBackgroundImage {
      bIsDynamicallyLoaded=false
      drawAs=NoDrawType
      imageType=NoImage
      mirroring=NoMirror
      resourceName=None
      resourceObject=null
      tiling=NoTile
      imageSize=<0 0>
      margin=<ltrb 0 0 0 0>
      outlineSettings {
       bUseBrushTransparency=false
       roundingType=HalfHeightRadius
       width=0
       color {colorUseRule=UseColor_Specified, specifiedColor=<rgba 0 0 0 0>}
       cornerRadii=<0 0 0 0>
      }
      tintColor {colorUseRule=UseColor_Specified, specifiedColor=<rgba 1 1 1 1>}
      uVRegion {bIsValid=false, max=<0 0>, min=<0 0>}
     }
     verticalBottomSlotImage {
      bIsDynamicallyLoaded=false
      drawAs=NoDrawType
      imageType=NoImage
      mirroring=NoMirror
      resourceName=None
      resourceObject=null
      tiling=NoTile
      imageSize=<0 0>
      margin=<ltrb 0 0 0 0>
      outlineSettings {
       bUseBrushTransparency=false
       roundingType=HalfHeightRadius
       width=0
       color {colorUseRule=UseColor_Specified, specifiedColor=<rgba 0 0 0 0>}
       cornerRadii=<0 0 0 0>
      }
      tintColor {colorUseRule=UseColor_Specified, specifiedColor=<rgba 1 1 1 1>}
      uVRegion {bIsValid=false, max=<0 0>, min=<0 0>}
     }
     verticalTopSlotImage {
      bIsDynamicallyLoaded=false
      drawAs=NoDrawType
      imageType=NoImage
      mirroring=NoMirror
      resourceName=None
      resourceObject=null
      tiling=NoTile
      imageSize=<0 0>
      margin=<ltrb 0 0 0 0>
      outlineSettings {
       bUseBrushTransparency=false
       roundingType=HalfHeightRadius
       width=0
       color {colorUseRule=UseColor_Specified, specifiedColor=<rgba 0 0 0 0>}
       cornerRadii=<0 0 0 0>
      }
      tintColor {colorUseRule=UseColor_Specified, specifiedColor=<rgba 1 1 1 1>}
      uVRegion {bIsValid=false, max=<0 0>, min=<0 0>}
     }
    }
    widgetStyle {
     barThickness=8
     bottomShadowBrush {
      bIsDynamicallyLoaded=false
      drawAs=Box
      imageType=FullColor
      mirroring=NoMirror
      resourceName=../../../Engine/Content/Slate/Common/ScrollBoxShadowBottom.png
      resourceObject=null
      tiling=NoTile
      imageSize=<16 8>
      margin=<ltrb 0.5 0 0.5 1>
      outlineSettings {
       bUseBrushTransparency=false
       roundingType=HalfHeightRadius
       width=0
       color {colorUseRule=UseColor_Specified, specifiedColor=<rgba 0 0 0 0>}
       cornerRadii=<0 0 0 0>
      }
      tintColor {colorUseRule=UseColor_Specified, specifiedColor=<rgba 1 1 1 1>}
      uVRegion {bIsValid=false, max=<0 0>, min=<0 0>}
     }
     horizontalScrolledContentPadding=<ltrb 0 0 1 0>
     leftShadowBrush {
      bIsDynamicallyLoaded=false
      drawAs=Box
      imageType=FullColor
      mirroring=NoMirror
      resourceName=../../../Engine/Content/Slate/Common/ScrollBoxShadowLeft.png
      resourceObject=null
      tiling=NoTile
      imageSize=<8 16>
      margin=<ltrb 1 0.5 0 0.5>
      outlineSettings {
       bUseBrushTransparency=false
       roundingType=HalfHeightRadius
       width=0
       color {colorUseRule=UseColor_Specified, specifiedColor=<rgba 0 0 0 0>}
       cornerRadii=<0 0 0 0>
      }
      tintColor {colorUseRule=UseColor_Specified, specifiedColor=<rgba 1 1 1 1>}
      uVRegion {bIsValid=false, max=<0 0>, min=<0 0>}
     }
     rightShadowBrush {
      bIsDynamicallyLoaded=false
      drawAs=Box
      imageType=FullColor
      mirroring=NoMirror
      resourceName=../../../Engine/Content/Slate/Common/ScrollBoxShadowRight.png
      resourceObject=null
      tiling=NoTile
      imageSize=<8 16>
      margin=<ltrb 0 0.5 1 0.5>
      outlineSettings {
       bUseBrushTransparency=false
       roundingType=HalfHeightRadius
       width=0
       color {colorUseRule=UseColor_Specified, specifiedColor=<rgba 0 0 0 0>}
       cornerRadii=<0 0 0 0>
      }
      tintColor {colorUseRule=UseColor_Specified, specifiedColor=<rgba 1 1 1 1>}
      uVRegion {bIsValid=false, max=<0 0>, min=<0 0>}
     }
     topShadowBrush {
      bIsDynamicallyLoaded=false
      drawAs=Box
      imageType=FullColor
      mirroring=NoMirror
      resourceName=../../../Engine/Content/Slate/Common/ScrollBoxShadowTop.png
      resourceObject=null
      tiling=NoTile
      imageSize=<16 8>
      margin=<ltrb 0.5 1 0.5 0>
      outlineSettings {
       bUseBrushTransparency=false
       roundingType=HalfHeightRadius
       width=0
       color {colorUseRule=UseColor_Specified, specifiedColor=<rgba 0 0 0 0>}
       cornerRadii=<0 0 0 0>
      }
      tintColor {colorUseRule=UseColor_Specified, specifiedColor=<rgba 1 1 1 1>}
      uVRegion {bIsValid=false, max=<0 0>, min=<0 0>}
     }
     verticalScrolledContentPadding=<ltrb 0 0 0 1>
    }
   }
   slot {class=SizeBoxSlot}
  }
  DetailsView_162 {
   widgetClassPath=&DetailsView
   parent=AthenaScrollBox_0
   widget=DetailsView_162
   properties {propertiesToShow [IDs]}
   slot {class=ScrollBoxSlot}
  }
  Button_GenerateVTIDs {widgetClassPath=&EditorUtilityButton, parent=HorizontalBox_116, widget=Button_GenerateVTIDs, slot {class=HorizontalBoxSlot}}
  TextBlock {
   widgetClassPath=&TextBlock
   parent=Button_GenerateVTIDs
   widget=TextBlock
   properties {text=Generate Variants}
   slot {class=ButtonSlot}
  }
  EditorUtilityEditableTextBox_376 {
   widgetClassPath=&EditorUtilityEditableTextBox
   bIsVariable=true
   parent=HorizontalBox_116
   widget=EditorUtilityEditableTextBox_376
   slot {class=HorizontalBoxSlot, properties {size {sizeRule=Fill, value=1}}}
  }
  TextBlock_204 {
   widgetClassPath=&TextBlock
   parent=VerticalBox_31
   widget=TextBlock_204
   properties {
    text="This tool takes cosmetic IDs and generate variant tokens for them based on their variant parameters.\r\nIt will ignore variant options that start unlocked and will work with mesh, particle and material variant items.\r\nFor now the VTIDs will be missing the Icon components and won't propagate the variant names."
    font {
     bForceMonospaced=false
     bMaterialIsStencil=false
     fontMaterial=null
     fontObject=&Roboto
     letterSpacing=0
     monospacedWidth=1
     size=10.5
     skewAmount=0
     typefaceFontName=Bold
     outlineSettings {
      bApplyOutlineToDropShadows=false
      bMiteredCorners=false
      bSeparateFillAlpha=false
      outlineMaterial=null
      outlineSize=0
      outlineColor=<rgba 0 0 0 1>
     }
    }
   }
   slot {class=VerticalBoxSlot}
  }
  EditorUtilityScrollBox_121 {
   widgetClassPath=&EditorUtilityScrollBox
   parent=VerticalBox_31
   widget=EditorUtilityScrollBox_121
   slot {class=VerticalBoxSlot, properties {size {sizeRule=Fill, value=1}}}
  }
  DetailsView {
   widgetClassPath=&DetailsView
   parent=EditorUtilityScrollBox_121
   widget=DetailsView
   properties {visibility=Collapsed}
   slot {class=ScrollBoxSlot, properties {size {sizeRule=Fill, value=1}}}
  }
  EditorUtilityEditableTextBox {
   widgetClassPath=&EditorUtilityEditableTextBox
   bIsVariable=true
   parent=EditorUtilityScrollBox_121
   widget=EditorUtilityEditableTextBox
   slot {class=ScrollBoxSlot, properties {size {sizeRule=Fill, value=1}}}
  }
 ]
}
